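# Widget Blueprint for the fake loading screen in Fortnite's guided tutorial, derived from the native GuidedTutorialLoadingScreen class.
# An overlay stacks a solid color background, an inherited native host slot for the loading screen and two close buttons.
# Hardware visibility borders show the top-right close button on touch input and the bottom-right Close button, bound to the Back action, otherwise.
# Its event graph contains only unconnected stub events, with no Blueprint logic.
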WidgetBlueprint /FortniteCoreUI/Athena/Playlists/Tutorial/GuidedTutorial_FakeLoadingScreen
graphs [EventGraph {nodes [
   Event PreConstruct()
   Event Construct()
   Event Tick()
  ]}]
variables []
widgets {
 info {inheritedWidgetCount=3, parentClass=/Script/FortniteUI.GuidedTutorialLoadingScreen, rootWidgetClass=&Overlay, widgetCount=8}
 widgets [
  Overlay_2 {widgetClassPath=&Overlay, widget=Overlay_2}
  BaseColor {
   widgetClassPath=&Image
   bIsVariable=true
   parent=Overlay_2
   widget=BaseColor
   properties {colorAndOpacity=<rgba 0.6445 0.0222 0 1>}
   slot {class=OverlaySlot, properties {horizontalAlignment=HAlign_Fill, verticalAlignment=VAlign_Fill}}
  }
  LoadingScreenSlot {
   widgetClassPath=/Script/UMG.NativeWidgetHost
   bInherited=true
   bIsVariable=true
   parent=Overlay_2
   widget=LoadingScreenSlot
   slot {class=OverlaySlot, properties {horizontalAlignment=HAlign_Fill, verticalAlignment=VAlign_Fill}}
  }
  VisibilityBorder_Touch {
   widgetClassPath=&CommonHardwareVisibilityBorder
   parent=Overlay_2
   widget=VisibilityBorder_Touch
   properties {visibilityQuery {
     autoDescription=" ANY( Input.Touch )"
     tokenStreamVersion=0
     userDescription=Touch
     queryTokenStream [0,1,1,1,0]
     tagDictionary [{tagName=Input.Touch}]
    }}
   slot {class=OverlaySlot, properties {horizontalAlignment=HAlign_Right, padding=<ltrb 0 30 30 0>}}
  }
  Button_CloseTouch {
   widgetClassPath=/FortUILibrary/Foundation/Buttons/CloseButton
   bInherited=true
   bIsVariable=true
   parent=VisibilityBorder_Touch
   widget=Button_CloseTouch
   properties {
    onButtonBaseClicked=(&self:WidgetTree.Button_CloseTouch.HandleOnButtonClicked)
    onButtonBaseDoubleClicked=(&self:WidgetTree.Button_CloseTouch.HandleOnButtonDoubleClicked)
    onButtonBaseHovered=(&self:WidgetTree.Button_CloseTouch.HandleOnButtonHovered)
    onButtonBaseUnhovered=(&self:WidgetTree.Button_CloseTouch.HandleOnButtonUnhovered)
    onSelectedChangedBase=(&self:WidgetTree.Button_CloseTouch.HandleOnSelectedChanged)
   }
   slot {class=BorderSlot, properties {padding=<ltrb 0 0 0 0>}}
  }
  OverlayCloseButton {
   widgetClassPath=&Overlay
   parent=Overlay_2
   widget=OverlayCloseButton
   slot {class=OverlaySlot, properties {horizontalAlignment=HAlign_Fill, verticalAlignment=VAlign_Fill}}
  }
  VisibilityBorder_NonTouch {
   widgetClassPath=&CommonHardwareVisibilityBorder
   parent=OverlayCloseButton
   widget=VisibilityBorder_NonTouch
   properties {visibilityQuery {
     autoDescription=" NONE( Input.Touch )"
     tokenStreamVersion=0
     userDescription=NonTouch
     queryTokenStream [0,1,3,1,0]
     tagDictionary [{tagName=Input.Touch}]
    }}
   slot {class=OverlaySlot, properties {horizontalAlignment=HAlign_Right, verticalAlignment=VAlign_Bottom, padding=<ltrb 0 0 60 175>}}
  }
  Button_Close {
   widgetClassPath=&IconTextButton
   bInherited=true
   bIsVariable=true
   parent=VisibilityBorder_NonTouch
   widget=Button_Close
   properties {
    buttonText=Close
    onButtonBaseClicked=(&self:WidgetTree.Button_Close.HandleOnButtonClicked)
    onButtonBaseDoubleClicked=(&self:WidgetTree.Button_Close.HandleOnButtonDoubleClicked)
    onButtonBaseHovered=(&self:WidgetTree.Button_Close.HandleOnButtonHovered)
    onButtonBaseUnhovered=(&self:WidgetTree.Button_Close.HandleOnButtonUnhovered)
    onSelectedChangedBase=(&self:WidgetTree.Button_Close.HandleOnSelectedChanged)
    style=/Game/UI/Frontend/ItemShop/Com/ButtonStyle-Com-LSkew.ButtonStyle-Com-LSkew_C
    triggeringInputAction {dataTable=&InputActionDataTable, rowName=Back}
   }
   slot {class=BorderSlot, properties {padding=<ltrb 0 0 0 0>}}
  }
 ]
}
